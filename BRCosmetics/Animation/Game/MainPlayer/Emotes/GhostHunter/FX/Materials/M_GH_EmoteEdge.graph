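# Additive, unlit, two-sided edge material for Niagara mesh particles in the GhostHunter emote effects.
# Color blends a particle-tinted base texture toward GridColor through flickering screen-space blocks with stepped panning, masked by vertex red, then feeds an emissive adjustment function.
# A WPO branch combines world-aligned noise, sine-driven glitch flicker offset by actor position, and a TurnOn-driven height pulse to displace vertices horizontally away from the object center.
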Material /BRCosmetics/Animation/Game/MainPlayer/Emotes/GhostHunter/FX/Materials/M_GH_EmoteEdge
graph {
 properties {
  bUsedWithNiagaraMeshParticles=true
  blendMode=BLEND_Additive
  pixelDepthOffsetMode=PDOM_Legacy
  shadingModel=MSM_Unlit
  twoSided=true
  usedShadingModels=Unlit
 }
 nodes [
  n1 = TextureSample() {texture=/Game/Packages/PBW/BASE/Textures/T_PBW_Base1_01}
  n2 = Power(Base=@n1.R, Exp=ScalarParameter() {defaultValue=4, parameterName=x}) {constExponent=4}
  n3 = Multiply(A=@n1.R, B=VectorParameter() {defaultValue=<rgba 1 1 1 0>}.B)
  n4 = ParticleColor()
  n5 = Multiply(A=ScreenPosition().ViewportUV) {constB=10}
  n6 = Time()
  n7 = ScalarParameter() {defaultValue=3, parameterName=Steps}
  n8 = Divide(A=Ceil(None=Multiply(A=@n6, B=@n7)), B=@n7)
  n9 = VectorParameter() {parameterName=BlocksPanSpeed, defaultValue=<rgba 1.8 -0.5 -0.8 0.6>}
  n10 = AppendVector(A=@n9.R, B=@n9.R)
  n11 = Multiply(A=Time())
  n12 = Ceil(None=Cosine(None=Sine(None=@n11) {period=2}) {period=0.5})
  n13 = ConstantBiasScale(None=@n11)
  n14 = Ceil(None=Sine(None=Cosine(None=@n13) {period=2}))
  n15 = Multiply(
   A=TextureSample(UVs=Panner(Coordinate=@n5, Time=@n8, Speed=@n10) {bFractionalPart=true}) {texture=/Game/Weapons/Prototype/ObjectMover/Textures/T_Movetool_Transition}.R
   B=VertexInterpolator(VS=Multiply(A=@n12, B=@n14)).PS
  )
  n16 = VertexColor()
  n17 = LinearInterpolate(
   A=Multiply(A=Add(A=@n2, B=@n3), B=@n4.RGB)
   B=VectorParameter() {parameterName=GridColor, defaultValue=<rgba 0 1 0.4121 0>}.RGB
   Alpha=Multiply(A=Saturate(None=@n15), B=@n16.R)
  )
  MaterialFunctionCall(Emissive=@n17) {materialFunction=&MF_FX_EmissiveAdjustments}
  Multiply(A=@n4.A, B=@n16.R)
  n18 = MaterialFunctionCall!(InVector=WorldPosition().XYZ) {materialFunction=/Game/Athena/Environments/Landscape/Foliage/MF_MakePreciseVec3}
  n19 = ObjectPositionWS()
  n20 = Multiply(A=Normalize(VectorInput=Subtract(A=@n18.Result, B=@n19)), B=Constant3Vector() {constant=<rgba 1 1 0 0>})
  n21 = Add(
   A=Multiply(
    A=ScalarParameter() {defaultValue=10, group=Noise, parameterName=WPOStrength}
    B=DynamicParameter() {defaultValue=<rgba 0 1 0 1>, paramNames [Block UV, WPO, FinalSaturation, Param4]}.WPO
   )
   B=Multiply(A=ObjectRadius(), B=ScalarParameter() {defaultValue=0.025, group=Noise, parameterName=WPOStrengthFromObjectRadius}) {constB=0.05}
  )
  n22 = TextureObject() {samplerType=SAMPLERTYPE_LinearColor, texture=&T_NoisePack_LinearColor}
  n23 = Divide(A=ObjectRadius(), B=ScalarParameter() {defaultValue=0.1, group=Noise, parameterName=NoiseSize})
  n24 = Multiply(A=VectorParameter() {group=Noise, parameterName=NoiseTileMult, defaultValue=<rgba 1 1 0.1 1>}.RGB, B=@n23)
  n25 = Time()
  n26 = VectorParameter() {group=Noise, parameterName=NoisePanSpeed, defaultValue=<rgba -0.08 0.07 -0.2 0>}
  n27 = Add(A=WorldPosition().XYZ, B=Multiply(A=@n23, B=Multiply(A=@n25, B=@n26.RGB)))
  n28 = Multiply(
   A=Power(Base=MaterialFunctionCall(TextureObject=@n22, TextureSize=@n24, WorldPosition=@n27) {materialFunction=&WorldAlignedTexture}.XY Texture) {constExponent=4}
   B=ScalarParameter() {defaultValue=2, group=Noise, parameterName=NoiseMaskMult}
  ) {constB=20}
  n29 = ActorPositionWS()
  n30 = MaterialFunctionCall(Float2=Round(None=Divide(A=@n29) {constB=2000})) {materialFunction=&BreakOutFloat2Components}
  n31 = Add(A=Time(), B=Add(A=@n30.R, B=@n30.R))
  n32 = Multiply(A=ScalarParameter() {defaultValue=3, parameterName=FlickerSpeed}, B=@n31)
  n33 = Sine(None=@n32) {period=1.6}
  n34 = Sine(None=@n32) {period=1.42}
  n35 = Ceil(None=AppendVector(A=AppendVector(A=@n33, B=@n34), B=Sine(None=@n32) {period=1.51}))
  n36 = ScalarParameter() {defaultValue=0.5, group=Noise, parameterName=NoiseFrequency}
  n37 = Add(A=Sine(None=@n32) {period=0.72}, B=Multiply(B=@n36) {constA=0.75}) {constB=0.75}
  n38 = Multiply(A=Add(A=Sine(None=@n32) {period=2.1}, B=Multiply(B=@n36) {constA=0.25}), B=@n37)
  n39 = Multiply(B=@n36) {constA=0.9, constB=0.9}
  n40 = Sine(None=@n32) {period=1.25}
  n41 = Multiply(A=@n37, B=Ceil(None=Add(A=@n39, B=@n40) {constB=0.9}))
  n42 = Sine(None=@n31) {period=0.075}
  n43 = Multiply(A=Saturate(None=Multiply(A=@n38, B=@n41)), B=LinearInterpolate(Alpha=ConstantBiasScale(None=@n42)) {constA=0.7})
  n44 = Add(
   A=ScalarParameter() {group=Noise, parameterName=NoiseAdd}
   B=Multiply(A=Multiply(A=@n35, B=@n43), B=ScalarParameter() {defaultValue=2, group=Noise, parameterName=NoiseMult})
  )
  n45 = Desaturation(None=Saturate(None=Multiply(A=@n28, B=@n44)))
  n46 = ComponentMask(None=WorldPosition().XYZ) {b=true}
  n47 = ComponentMask(None=MaterialFunctionCall() {materialFunction=&ObjectPivotPoint}.Mesh Particle Pivot Location) {b=true}
  n48 = ScalarParameter() {defaultValue=1, parameterName=TurnOn}
  n49 = LinearInterpolate(A=@n47, B=Add(A=@n47, B=ScalarParameter() {defaultValue=300, parameterName=PulseHeight}) {constB=1600}, Alpha=@n48) {constA=-0.25, constB=1.2}
  n50 = ScalarParameter() {defaultValue=33, parameterName=PulseRadius}
  n51 = LinearInterpolate(A=@n50, B=Multiply(A=@n50) {constB=0.1}, Alpha=@n48) {constA=500, constB=100}
  n52 = Multiply(A=Multiply(A=SphereMask(A=@n46, B=@n49, Radius=@n51) {hardnessPercent=0}, B=OneMinus(None=@n48))) {constB=2}
  NamedRerouteDeclaration(None=Multiply(A=Add(A=@n45, B=@n52), B=Multiply(A=@n20, B=@n21)) {constB=0}) {name=WPO, nodeColor=<rgba 0 1 0.9297 1>}
  NamedRerouteDeclaration(None=@n17) {name=BaseColor, nodeColor=<rgba 0.0078 1 0 1>}
 ]
}
