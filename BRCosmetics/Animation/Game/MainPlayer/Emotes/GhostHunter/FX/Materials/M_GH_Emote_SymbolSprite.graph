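# Additive unlit Niagara sprite material, named for the symbol sprites in the Ghost Hunter emote.
# A panning DFHeightMask sample slightly distorts particle SubUV lookups into the same texture, whose red channel is smoothstepped using vector-parameter thresholds.
# That mask, scaled by Core, adds to particle color and feeds MF_FX_EmissiveAdjustments.
# The mask is also multiplied by the panning noise minus inverted particle alpha, producing an erosion-style fade.
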Material /BRCosmetics/Animation/Game/MainPlayer/Emotes/GhostHunter/FX/Materials/M_GH_Emote_SymbolSprite
graph {
 properties {bUsedWithNiagaraSprites=true, blendMode=BLEND_Additive, pixelDepthOffsetMode=PDOM_Legacy, shadingModel=MSM_Unlit, usedShadingModels=Unlit}
 nodes [
  n1 = VectorParameter() {parameterName="A,W,B,W", defaultValue=<rgba 0 0 0 1>}
  n2 = AppendVector(A=@n1.R, B=@n1.R)
  n3 = TextureSample(UVs=Panner(Coordinate=TextureCoordinate() {uTiling=0.2, vTiling=0.2}) {speedY=0.1}) {samplerType=SAMPLERTYPE_LinearColor, texture=/Game/Athena/Prototype/Blueprints/Cube/Materials/DFHeightMask}
  n4 = ParticleSubUVProperties()
  n5 = TextureSample(UVs=Add(A=Multiply(A=@n3.B) {constB=0.02}, B=LinearInterpolate(A=@n4.TextureCoordinate0, B=@n4.TextureCoordinate0, Alpha=@n4.TextureCoordinate0))) {samplerType=SAMPLERTYPE_LinearColor, texture=&DFHeightMask}
  n6 = SmoothStep(Min=@n2, Max=Add(A=@n2, B=AppendVector(A=@n1.G, B=@n1.G)), Value=AppendVector(A=@n5.R, B=@n5.R))
  n7 = ComponentMask(None=@n6) {r=true}
  n8 = ScalarParameter() {parameterName=Core}
  MaterialFunctionCall(Emissive=Add(A=ParticleColor().RGB, B=Multiply(A=@n7, B=@n8))) {materialFunction=&MF_FX_EmissiveAdjustments}
  n9 = OneMinus(None=ParticleColor().A)
  Multiply(A=ComponentMask(None=@n6) {g=true}, B=Saturate(None=Subtract(A=@n3.B, B=@n9)))
 ]
}
