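# Additive unlit material for Niagara mesh particles in the GhostHunter emote, drawing a blocky screen-space flash.
# Panning noise drives DDX/DDY red and green edges, added to particle color and scaled by scene-to-pixel depth difference for emissive.
# Scene depth reconstructs surface positions in particle space, yielding a box volume mask and Z gradient.
# A smoothstep block ramp thresholds the noise using particle alpha and smoke texture.
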Material /BRCosmetics/Animation/Game/MainPlayer/Emotes/GhostHunter/FX/Materials/M_GH_FlashChange
graph {
 properties {
  bUsedWithNiagaraMeshParticles=true
  blendMode=BLEND_Additive
  pixelDepthOffsetMode=PDOM_Legacy
  shadingModel=MSM_Unlit
  usedShadingModels=Unlit
 }
 nodes [
  n1 = ScreenPosition()
  n2 = ScalarParameter() {defaultValue=10, group=Block, parameterName=BlockUV}
  n3 = Time()
  n4 = ScalarParameter() {defaultValue=10.7749, group=Block, parameterName=Block U Speed}
  n5 = TextureSample(UVs=Panner(Coordinate=Multiply(A=@n1.ViewportUV, B=@n2), Time=Multiply(A=@n3, B=@n4) {constB=0.5}) {desc="-0.33,-0.43", speedY=0.1}) {texture=/Game/Athena/UI/Challenges/Art/T_UI_DataChip_Noise}
  n6 = TextureSample(UVs=ScreenPosition().ViewportUV) {texture=&T_Inky_Smoke_Tile}
  n7 = LinearInterpolate(A=@n5.R, B=@n5.R, Alpha=@n6.G)
  n8 = Add(A=ParticleColor().RGB, B=Multiply(A=Constant3Vector() {constant=<rgba 3 0 0 0>}, B=DDX(Value=@n7)))
  n9 = Multiply(A=DDY(Value=@n7), B=Constant3Vector() {constant=<rgba 0 3 0 0>})
  n10 = SceneDepth()
  n11 = PixelDepth()
  n12 = Abs(None=Subtract(A=@n10.R, B=@n11.R))
  n13 = ScalarParameter() {parameterName=depth_look}
  MaterialFunctionCall(Emissive=Multiply(A=Add(A=@n8, B=@n9), B=Divide(A=@n12, B=@n13) {constB=20})) {materialFunction=&MF_FX_EmissiveAdjustments}
  Multiply(A=NamedRerouteUsage(), B=Multiply(A=ScalarParameter() {parameterName=T}, B=NamedRerouteUsage()))
  n14 = Divide(A=@n10.R, B=@n11.R)
  n15 = CameraPositionWS()
  n16 = Subtract(A=WorldPosition().XYZ, B=@n15)
  n17 = TransformPosition(None=Add(A=Multiply(A=@n14, B=@n16), B=@n15)) {transformSourceType=TRANSFORMPOSSOURCE_World, transformType=TRANSFORMPOSSOURCE_Particle}
  NamedRerouteDeclaration(None=ComponentMask(None=@n17) {b=true}) {name=ZGradient, nodeColor=<rgba 1 0.6563 0 1>}
  n18 = ParticleColor()
  n19 = OneMinus(None=Add(A=Add(A=@n18.A, B=@n6.G), B=NamedRerouteUsage()))
  n20 = ScalarParameter() {parameterName=Alpha_W}
  NamedRerouteDeclaration(None=SmoothStep(Min=@n19, Max=Add(A=@n19, B=@n20), Value=@n7)) {name=BlocksAlphaRamp, nodeColor=<rgba 1 0 0.1357 1>}
  NamedRerouteDeclaration(None=MaterialFunctionCall!(A=@n17, B, Bounds=Constant3Vector() {constant=<rgba 1 1 1 0>}, Edge Falloff=ScalarParameter() {parameterName=f}) {materialFunction=/Engine/Functions/Engine_MaterialFunctions01/Gradient/BoxMask-3D}.Result) {name=VolumeMask, nodeColor=<rgba 1.7485 0 3 1>}
 ]
}
