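# Translucent unlit Niagara mesh-particle material for the GhostHunter emote, apparently wiping a glitchy effect over a skin switch.
# It reconstructs the scene position behind the mesh from depth in particle space, feeding a box volume mask.
# A mask texture, distorted and roughened by smoke on higher quality levels, shapes the wipe within that volume.
# Screen-space stepped noise blocks and scrolling lines modulate a color ramp, with particle alpha driving block dissolve.
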Material /BRCosmetics/Animation/Game/MainPlayer/Emotes/GhostHunter/FX/Materials/M_GH_WipeOver
graph {
 properties {
  bUsedWithNiagaraMeshParticles=true
  blendMode=BLEND_Translucent
  pixelDepthOffsetMode=PDOM_Legacy
  shadingModel=MSM_Unlit
  usedShadingModels=Unlit
 }
 nodes [
  MaterialFunctionCall(Emissive=Add(A=NamedRerouteUsage(), B=Constant())) {materialFunction=&MF_FX_EmissiveAdjustments}
  Multiply(A=NamedRerouteUsage(), B=NamedRerouteUsage())
  n1 = OneMinus(None=ParticleColor().A)
  n2 = ScalarParameter() {parameterName=Alpha_W}
  n3 = ScreenPosition()
  n4 = ScalarParameter() {defaultValue=10, group=Block, parameterName=BlockUV}
  n5 = Multiply(A=Time(), B=ScalarParameter() {defaultValue=10.7749, group=Block, parameterName=Block U Speed}) {constB=0.5}
  n6 = TextureSample(UVs=Panner(Coordinate=Multiply(A=@n3.ViewportUV, B=@n4), Time=Ceil(None=@n5)) {desc="-0.33,-0.43", speedX=-0.3, speedY=0.5}) {texture=/Game/Athena/UI/Challenges/Art/T_UI_DataChip_Noise}
  NamedRerouteDeclaration(None=SmoothStep(Min=@n1, Max=Add(A=@n1, B=@n2), Value=@n6.R)) {name=BlocksAlphaRamp, nodeColor=<rgba 1 0 0.1357 1>}
  n7 = Divide(A=SceneDepth().R, B=PixelDepth().R)
  n8 = CameraPositionWS()
  n9 = Subtract(A=WorldPosition().XYZ, B=@n8)
  n10 = TransformPosition(None=Add(A=Multiply(A=@n7, B=@n9), B=@n8)) {transformSourceType=TRANSFORMPOSSOURCE_World, transformType=TRANSFORMPOSSOURCE_Particle}
  n11 = MaterialFunctionCall!(
   A=@n10
   B
   Bounds=VectorParameter() {group=Masking, parameterName=BoxMaskBounds, defaultValue=<rgba 0 0 0 1>}.RGB
   Edge Falloff=ScalarParameter() {group=Masking, parameterName=BoxMask_Falloff}
  ) {materialFunction=/Engine/Functions/Engine_MaterialFunctions01/Gradient/BoxMask-3D}
  n12 = ComponentMask(None=NamedRerouteUsage()) {r=true}
  n13 = NamedRerouteUsage()
  n14 = Multiply(A=Add(A=Multiply(A=@n13) {constB=2}) {constB=-1}, B=ScalarParameter() {parameterName=Blocks_DistortionStr})
  n15 = ComponentMask(None=NamedRerouteUsage()) {b=true, r=true}
  n16 = Constant2Vector() {g=2, r=0.5}
  n17 = TextureSample(UVs=Panner(Coordinate=Multiply(A=@n15, B=@n16) {constB=3}) {speedX=1}) {texture=&T_Inky_Smoke_Tile}
  n18 = Multiply(A=Subtract(A=Multiply(A=@n17.G) {constB=2})) {constB=0.2}
  n19 = Constant()
  n20 = TextureSample(UVs=Add(A=Add(A=@n12, B=@n14), B=QualitySwitch(Default=@n18, Low=@n19, High=@n18, Medium=@n19, Epic=@n18))) {
   samplerSource=SSM_Clamp_WorldGroupSettings
   samplerType=SAMPLERTYPE_Masks
   texture=/BRCosmetics/Animation/Game/MainPlayer/Emotes/GhostHunter/FX/Textures/T_GH_SkinSwitchMaskTime
  }
  NamedRerouteDeclaration(None=Multiply(A=NamedRerouteDeclaration(None=@n11.Result) {name=VolumeMask, nodeColor=<rgba 1.7485 0 3 1>}, B=@n20.R)) {name=WipeMask, nodeColor=<rgba 1.7485 0 3 1>}
  n21 = TextureSampleParameter2D(UVs=NamedRerouteUsage()) {
   group=Block
   parameterName=Block Color
   texture=/BRCosmetics/Animation/Game/MainPlayer/Emotes/GhostHunter/FX/Textures/T_GH_SkinSwitch_ColorRamp
  }
  n22 = MaterialFunctionCall(Float2=ScreenPosition().ViewportUV) {materialFunction=&BreakOutFloat2Components}
  n23 = ScalarParameter() {defaultValue=50, group=Block, parameterName=Line amount}
  n24 = Time()
  n25 = ScalarParameter() {defaultValue=0.2, group=Block, parameterName=Line Speed}
  n26 = Frac(None=Add(A=Multiply(A=@n22.G, B=@n23), B=Multiply(A=@n24, B=@n25)))
  n27 = Multiply(A=@n26, B=OneMinus(None=@n26))
  n28 = ScalarParameter() {defaultValue=-0.15, group=Block, parameterName=Line Size}
  n29 = LinearInterpolate(
   A=ScalarParameter() {group=Block, parameterName=Line Dark}
   B=ScalarParameter() {defaultValue=1, group=Block, parameterName=Line Light}
   Alpha=Ceil(None=Add(A=@n27, B=@n28))
  )
  n30 = Constant() {r=0.5}
  n31 = NamedRerouteDeclaration(None=StaticSwitchParameter(True=LinearInterpolate(Alpha=@n6.R) {constA=-0.2, constB=2}, False=Constant() {r=1}) {defaultValue=true, parameterName=bUseBlocks}) {name=Blocks, nodeColor=<rgba 1 0 0.1357 1>}
  n32 = ScalarParameter() {defaultValue=1, parameterName=BlockMultiply}
  n33 = Add(
   A=StaticSwitchParameter(True=Multiply(A=@n29, B=@n30), False=Constant()) {defaultValue=true, parameterName=bUseLines}
   B=StaticSwitchParameter(True=Multiply(A=@n31, B=@n32), False=Constant() {r=1}) {defaultValue=true, parameterName=bUseBlocks}
  )
  n34 = Multiply(A=Add(A=@n30, B=Constant()), B=Constant() {r=1})
  NamedRerouteDeclaration(None=Multiply(A=@n21.RGB, B=Add(A=@n33, B=@n34))) {name=BlockEmi, nodeColor=<rgba 1 0 0.1357 1>}
  n35 = ComponentMask(None=@n10) {b=true, g=true, r=true}
  NamedRerouteDeclaration(None=Saturate(None=Add(A=@n35) {constB=0.5})) {name=MeshParticle_LocalPositionBehindTransparency, nodeColor=<rgba 1.7485 0 3 1>}
 ]
}
