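# Niagara effect for the GhostHunter emote, named for the gadget closedown: a brief purple glow that collapses and fades.
# A single Glow emitter burst-spawns sprites that live under a second.
# Over each particle's life, curves dim the HDR magenta-to-blue color, fade alpha out and shrink the sprite to zero by mid-life.
# The sprite also slowly rotates and carries a small negative camera offset.
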NiagaraSystem /BRCosmetics/Animation/Game/MainPlayer/Emotes/GhostHunter/FX/NS_Emote_GhostHunter_GadgetClosedown
dependencies {
 usedDataInterfaces [&NiagaraDataInterfaceColorCurve, &NiagaraDataInterfaceCurve, &NiagaraDataInterfaceVector2DCurve]
 usedDynamicInputs []
 usedModules [&SystemState]
}
emitters {Glow {
  emitterUpdateScript [
   &EmitterState()
   &SpawnBurst_Instantaneous(Use Spawn Probability=false, UseLoopCountLimit=false)
  ]
  particleSpawnScript [&InitializeParticle(
    Color Mode=::NewEnumerator0
    Lifetime=0.7
    Sprite Size=<150 150>
    Sprite Size Mode=::NewEnumerator4
    UsePositionOffset=false
    Write Material Random=true
   )]
  particleUpdateScript [
   &ParticleState()
   &Color(Color=&ColorFromCurve(ColorCurve=DataInterface() {propertyValues {
       AlphaCurve {keys [<at 0 1 RCIM_Cubic>,<at 1 0 RCIM_Cubic>]}
       BlueCurve {keys [<at 0.2508 4.8786>,<at 0.8634 1.5>]}
       GreenCurve {keys [<at 0.2508 0>,<at 0.8634 0>]}
       RedCurve {keys [<at 0.2508 5>,<at 0.8634 0.528>]}
      }}))
   &ScaleSpriteSize(Uniform Curve Sprite Scale=DataInterface() {propertyValues {Curve {keys [
        {arriveTangent=-2.4678, interpMode=RCIM_Cubic, leaveTangent=-2.4678, tangentMode=RCTM_User, time=0, value=1}
        {arriveTangent=-0.4899, interpMode=RCIM_Cubic, leaveTangent=-0.4899, tangentMode=RCTM_User, time=0.5, value=0}
       ]}}})
   &CameraOffset(Camera Offset Amount=-10)
   &SpriteRotationRate(Rotation Rate=30)
  ]
  rendererClasses [&NiagaraSpriteRendererProperties]
  renderers [{rendererClass=&NiagaraSpriteRendererProperties}]
 }}
summary {
 systemName=NS_Emote_GhostHunter_GadgetClosedown
 emitters [{emitterName=Glow, rendererClasses [&NiagaraSpriteRendererProperties]}]
 userVariables []
}
