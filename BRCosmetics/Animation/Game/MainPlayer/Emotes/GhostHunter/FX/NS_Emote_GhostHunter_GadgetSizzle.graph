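# Niagara system for the GhostHunter emote that makes a gadget sizzle with glitchy magenta and violet particles, emitted from the User.SkeletalMesh input.
# CubeEdge meshes and Glitch001 sprites re-sample mesh location each update, with CubeEdge's spawn rate following a curve over its loop.
# The "Empty" emitter throws small tumbling meshes against sampled mesh velocity, slowed by drag.
# Symbols sprites drift with SubUV frames and shrink late in the loop via Emitter.SymbolsReturn.
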NiagaraSystem /BRCosmetics/Animation/Game/MainPlayer/Emotes/GhostHunter/FX/NS_Emote_GhostHunter_GadgetSizzle
dependencies {
 usedDataInterfaces [
  &NiagaraDataInterfaceCurve
  &NiagaraDataInterfaceMeshRendererInfo
  &NiagaraDataInterfaceArrayFloat3
  &NiagaraDataInterfaceArrayFloat
  &NiagaraDataInterfaceColorCurve
 ]
 usedDynamicInputs []
 usedModules [&SystemState]
}
emitters {
 CubeEdge {
  emitterUpdateScript [
   &EmitterState(Life Cycle Mode=::NewEnumerator1, Loop Behavior=::NewEnumerator1, Loop Duration=3.5, UseLoopDelay=false)
   &SpawnRate(
    SpawnRate=&FloatFromCurve(
     FloatCurve=DataInterface() {propertyValues {Curve {keys [
         {arriveTangent=1, interpMode=RCIM_Cubic, time=0.1, value=0.85}
         {arriveTangent=-1.1, leaveTangent=-1.1, time=0.3, value=0.3}
         {arriveTangent=1.3, leaveTangent=1.3, time=0.7, value=0.3}
         {arriveTangent=2.5833, interpMode=RCIM_Cubic, leaveTangent=2.5833, time=0.8, value=0.95}
         <at 1 1.075 RCIM_Cubic>
        ]}}}
     CurveIndex=$Emitter.NormalizedLoopAge
     Scale Curve=10
    )
    Use Spawn Probability=false
   )
  ]
  particleSpawnScript [
   &InitializeParticle(
    Color=<rgba 1.976 0 2 1>
    Lifetime Max=0.4
    Lifetime Min=0.2
    Lifetime Mode=::NewEnumerator1
    Mesh Scale Mode=::NewEnumerator1
    Mesh Uniform Scale Max=0.7
    Mesh Uniform Scale Min=0.5
    Sprite Size Max=<15 15>
    Sprite Size Min=<7 7>
    Sprite Size Mode=::NewEnumerator2
    UsePositionOffset=false
    Write Material Random=true
   ) {module=InitializeParticle001}
   &CameraOffset(Camera Offset Amount=&RandomRangeFloat(Minimum=0, Maximum=0))
   &DynamicMaterialParameters(
    Index 0 Param 1=&RandomRangeFloat(Minimum=0, Maximum=1)
    Index 0 Param 2=&RandomRangeFloat(Minimum=0, Maximum=0.5)
    Index 0 Param 3=-3
    Index 0 Param 4=5
    Module.Param0WriteEnabled=true
    Module.Param1WriteEnabled=true
    Module.Param2WriteEnabled=true
    Module.Param3WriteEnabled=true
   )
   &InitialMeshOrientation(
    Mesh Orientation Mode=::NewEnumerator6
    Orientation Quaternion:Quat4f=&MakeQuatFromAxisAngle(
     Axis=&SelectVectorFromArray(
      Array Sampling Mode=::NewEnumerator0
      Direct Array Index=?
      Vector Selection Array=DataInterface() {propertyValues {FloatData [<0 1 0>, <1 0 0>, <0 0 1>]}}
      Random Seed=0
     )
     AngleInDegrees=&SelectFloatFromArray(
      Array Sampling Mode=::NewEnumerator0
      Direct Array Index=?
      Float Selection Array=DataInterface() {propertyValues {FloatData [0,90,270,180]}}
      Random Seed=0
     )
    )
    Use Rotation Vector=false
   )
  ]
  particleUpdateScript [
   &ParticleState()
   &SkeletalMeshLocation(Bone Sampling Mode=::NewEnumerator0, MaskSpawnGroup=false, OverrideRandomSeed=false, Skeletal Mesh=$User.SkeletalMesh)
   &ScaleSpriteSize(Scale Sprite Size Mode=::NewEnumerator0, Uniform Scale Factor=&Multiply_Float(A=1, B=1)) {module=ScaleSpriteSize001}
   &ScaleMeshSize(Scale Factor=&VectorFromFloat(Value=&FloatFromCurve(FloatCurve=DataInterface() {propertyValues {Curve {keys [
          <at 0 1.125 RCIM_Cubic>
          {arriveTangent=0.75, interpMode=RCIM_Cubic, leaveTangent=0.75, time=0.3, value=1}
          <at 0.5 1.5 RCIM_Constant>
          <at 1 2 RCIM_Constant>
         ]}}})))
   &Color(Color=&ColorFromCurve(ColorCurve=DataInterface() {propertyValues {
       AlphaCurve {keys [<at 0 1 RCIM_Cubic>,<at 0.702 1>,<at 1 0 RCIM_Cubic>]}
       BlueCurve {keys [<at 0 20 RCIM_Cubic>,<at 0.149 25>,<at 0.7727 1.7901>,<at 1 0.7047 RCIM_Cubic>]}
       GreenCurve {keys [<at 0 0 RCIM_Cubic>,<at 0.149 0>,<at 0.7727 0>,<at 1 1 RCIM_Cubic>]}
       RedCurve {keys [<at 0 11.8725 RCIM_Cubic>,<at 0.149 14.8407>,<at 0.7727 2>,<at 1 0 RCIM_Cubic>]}
      }}))
  ]
  rendererClasses [&NiagaraMeshRendererProperties]
  renderers [{rendererClass=&NiagaraMeshRendererProperties, rendererIndex=0}]
 }
 Empty {
  emitterUpdateScript [
   &EmitterState()
   &SpawnRate(SpawnRate=6, Use Spawn Probability=false)
  ]
  particleSpawnScript [
   &InitializeParticle(
    Color Maximum=&RandomRangeLinearColor(Minimum=<rgba 0.4167 0 1 2>, Maximum=<rgba 0.95 0 1 2>)
    Color Minimum=&RandomRangeLinearColor(Minimum=<rgba 0.0667 1 0 2>, Maximum=<rgba 0 0.3833 1 2>)
    Color Mode=::NewEnumerator2
    Lifetime Max=0.5
    Lifetime Min=0.3
    Lifetime Mode=::NewEnumerator1
    Mesh Scale Mode=::NewEnumerator1
    Mesh Uniform Scale Max=0.05
    Mesh Uniform Scale Min=0.02
    Sprite Size Mode=::NewEnumerator1
    Uniform Sprite Size Max=4
    Uniform Sprite Size Min=2.5
    UsePositionOffset=false
    Write Material Random=true
   )
   &SkeletalMeshLocation(Bone Sampling Mode=::NewEnumerator0, MaskSpawnGroup=false, OverrideRandomSeed=false, Skeletal Mesh=$User.SkeletalMesh)
   &InitializeParticle(
    Color Maximum=&RandomRangeLinearColor(Minimum=<rgba 0.4167 0 1 2>, Maximum=<rgba 0.95 0 1 2>)
    Color Minimum=&RandomRangeLinearColor(Minimum=<rgba 0.0667 1 0 2>, Maximum=<rgba 0 0.3833 1 2>)
    Color Mode=::NewEnumerator2
    Lifetime Max=1.6
    Lifetime Min=0.7
    Lifetime Mode=::NewEnumerator1
    Mesh Scale Max=<0.08 0.08 0.08>
    Mesh Scale Min=<0.03 0.03 0.03>
    Mesh Scale Mode=::NewEnumerator2
    UsePositionOffset=false
    Write Material Random=true
   ) {enabled=false, module=InitializeParticle001}
   &DynamicMaterialParameters(
    Index 0 Param 1=&RandomRangeFloat(Minimum=0, Maximum=1)
    Index 0 Param 2=&UniformAOrBFloat(A=0)
    Index 0 Param 3=&UniformAOrBFloat(A=0, B=0.5)
    Index 0 Param 4=&RandomRangeFloat(Minimum=0.5, Maximum=1)
    Module.Param0WriteEnabled=true
    Module.Param1WriteEnabled=true
    Module.Param2WriteEnabled=true
    Module.Param3WriteEnabled=true
   )
  ]
  particleUpdateScript [
   &ParticleState()
   &SampleSkeletalMesh(
    AdditionalYawPitchRoll=false
    Bone Sampling Mode:ENiagaraBoneSamplingMode=::NewEnumerator0
    OverrideRandomSeed=false
    Skeletal Mesh:NiagaraDataInterfaceSkeletalMesh=$User.SkeletalMesh
   )
   &ScaleMeshSize(Scale Factor=&VectorFromFloat(Value=&FloatFromCurve(
      FloatCurve=DataInterface() {propertyValues {Curve {keys [
          <at 0 0 RCIM_Cubic>
          {arriveTangent=0.625, interpMode=RCIM_Cubic, leaveTangent=0.625, time=0.2, value=1}
          {arriveTangent=-0.5, interpMode=RCIM_Cubic, leaveTangent=-0.5, time=0.4, value=0.25}
          {arriveTangent=-0.4167, interpMode=RCIM_Cubic, leaveTangent=-0.4167, time=0.7, value=0.75}
          <at 1 0 RCIM_Cubic>
         ]}}}
      Scale Curve=3
     )))
   /SetVariables(Particles.Velocity=&Multiply_VectorByFloat(Vector=$StackContext.SkeletalMeshLocation.SampledVelocity, Float=-1)) {isSetParameters=true}
   &AccelerationForce(Acceleration=&RandomRangeVector(Minimum=<0 0 -50>, Maximum=<1 1 50>))
   &UpdateMeshOrientation(
    Rotation Rate=&RandomRangeFloat(Minimum=1, Maximum=1)
    Rotation Vector=&RandomRangeVector(Minimum=<-1 -1 -1>, Maximum=<1 1 1>)
   )
   &InheritSourceMovement(Applied Position Scale=<0 0 0>, ApplyForce=false, ApplyPosition=true, ApplyPositionFalloff=false) {enabled=false}
   &Drag(Drag=3, Use Linear Drag=true, Use Rotational Drag=false)
   &SolveForcesAndVelocity(Clamp Velocity=false, Limit Acceleration=false)
  ]
  rendererClasses [&NiagaraMeshRendererProperties]
  renderers [{rendererClass=&NiagaraMeshRendererProperties, rendererIndex=0}]
 }
 Glitch001 {
  emitterUpdateScript [
   &EmitterState(Life Cycle Mode=::NewEnumerator1, Loop Behavior=::NewEnumerator1, UseLoopDelay=false)
   &SpawnRate(SpawnRate=7, Use Spawn Probability=false)
  ]
  particleSpawnScript [
   &InitializeParticle(
    Color=<rgba 0 0 0 1>
    Lifetime Max=0.5
    Lifetime Min=0.3
    Lifetime Mode=::NewEnumerator1
    Sprite Size Mode=::NewEnumerator2
    UsePositionOffset=false
    Write Material Random=true
   ) {module=InitializeParticle001}
   &CameraOffset(Camera Offset Amount=&RandomRangeFloat(Minimum=0, Maximum=0))
   &DynamicMaterialParameters(
    Index 0 Param 1=&RandomRangeFloat(Minimum=0, Maximum=1)
    Index 0 Param 2=&RandomRangeFloat(Minimum=0, Maximum=0.5)
    Index 0 Param 3=-3
    Index 0 Param 4=5
    Module.Param0WriteEnabled=true
    Module.Param1WriteEnabled=true
    Module.Param2WriteEnabled=true
    Module.Param3WriteEnabled=true
   )
  ]
  particleUpdateScript [
   &ParticleState()
   &SkeletalMeshLocation(Bone Sampling Mode=::NewEnumerator0, MaskSpawnGroup=false, OverrideRandomSeed=false, Skeletal Mesh=$User.SkeletalMesh)
   &ScaleSpriteSize(Scale Factor=&MakeVector2D(
     X=&FloatFromCurve(FloatCurve=DataInterface() {propertyValues {Curve {keys [<at 0 1.5>,<at 1 0.5>]}}})
     Y=&FloatFromCurve(FloatCurve=DataInterface() {propertyValues {Curve {keys [<at 0 1>,<at 1 2>]}}})
    ))
   &SolveForcesAndVelocity(Clamp Velocity=false, Limit Acceleration=false)
   &ScaleSpriteSize(Scale Sprite Size Mode=::NewEnumerator0, Uniform Scale Factor=&Multiply_Float(A=1, B=1)) {module=ScaleSpriteSize001}
  ]
  rendererClasses [&NiagaraSpriteRendererProperties]
  renderers [{rendererClass=&NiagaraSpriteRendererProperties}]
 }
 Symbols {
  emitterUpdateScript [
   &EmitterState(Life Cycle Mode=::NewEnumerator1, Loop Behavior=::NewEnumerator1, Loop Duration=3.4, UseLoopDelay=false)
   &SpawnRate(SpawnRate=8, Use Spawn Probability=false)
   /SetVariables(Emitter.SymbolsReturn:NiagaraFloat=&FloatFromCurve(
     FloatCurve=DataInterface() {propertyValues {Curve {keys [<at 0 0>, {arriveTangent=1.1, leaveTangent=1.1, time=0.8, value=0}, <at 1 1.1 RCIM_Cubic>]}}}
     CurveIndex=$Emitter.NormalizedLoopAge
    )) {isSetParameters=true}
  ]
  particleSpawnScript [
   &InitializeParticle(
    Color=<rgba 1 0 0.6189 1>
    Lifetime Max=0.8
    Lifetime Min=0.5
    Lifetime Mode=::NewEnumerator1
    Sprite Size Min=<7 7>
    Sprite Size Mode=::NewEnumerator2
    UsePositionOffset=false
    Write Material Random=true
   ) {module=InitializeParticle001}
   &CameraOffset(Camera Offset Amount=&RandomRangeFloat(Minimum=0, Maximum=0))
   &DynamicMaterialParameters(
    Index 0 Param 1=&RandomRangeFloat(Minimum=0, Maximum=1)
    Index 0 Param 2=&RandomRangeFloat(Minimum=0, Maximum=0.5)
    Index 0 Param 3=-3
    Index 0 Param 4=5
    Module.Param0WriteEnabled=true
    Module.Param1WriteEnabled=true
    Module.Param2WriteEnabled=true
    Module.Param3WriteEnabled=true
   )
   &SubUVAnimation(End Frame=3, SubUV Animation Mode:ENiagara_SubUVLookupMode=::NewEnumerator2)
   &SkeletalMeshLocation(Bone Sampling Mode=::NewEnumerator0, MaskSpawnGroup=false, OverrideRandomSeed=false, Skeletal Mesh=$User.SkeletalMesh)
  ]
  particleUpdateScript [
   &ParticleState()
   &ScaleSpriteSize(Scale Factor=&MakeVector2D(
     X=&FloatFromCurve(FloatCurve=DataInterface() {propertyValues {Curve {keys [<at 0 1.025>,<at 1 0.625>]}}})
     Y=&FloatFromCurve(FloatCurve=DataInterface() {propertyValues {Curve {keys [<at 0 1>,<at 1 0.875>]}}})
    ))
   &SkeletalMeshLocation(Bone Sampling Mode=::NewEnumerator0, MaskSpawnGroup=false, OverrideRandomSeed=false, Skeletal Mesh=$User.SkeletalMesh) {enabled=false, module=SkeletalMeshLocation001}
   &InheritSourceMovement(
    Applied Position Scale=&VectorFromFloat(Value=&FloatFromCurve(FloatCurve=DataInterface() {propertyValues {Curve {keys [<at 0 0.275 RCIM_Cubic>,<at 1 0 RCIM_Cubic>]}}}, Scale Curve=0))
    ApplyForce=false
    ApplyPosition=true
    ApplyPositionFalloff=false
   )
   &AccelerationForce(Acceleration=&RandomRangeVector(Minimum=<0 0 20>, Maximum=<0 1 50>))
   &SolveForcesAndVelocity(Clamp Velocity=false, Limit Acceleration=false)
   &ScaleSpriteSize(
    Scale Sprite Size Mode=::NewEnumerator0
    Uniform Scale Factor=&FloatFromCurve(FloatCurve=DataInterface() {propertyValues {Curve {keys [<at 0 1>,<at 1 0>]}}}, CurveIndex=$Emitter.SymbolsReturn)
   ) {module=ScaleSpriteSize001}
   &Color(Color=&ColorFromCurve(ColorCurve=DataInterface() {propertyValues {
       AlphaCurve {keys [
         <at 0 0 RCIM_Cubic>
         {arriveTangent=2.2959, interpMode=RCIM_Cubic, leaveTangent=2.2959, time=0.2, value=1}
         {arriveTangent=-1.25, interpMode=RCIM_Cubic, leaveTangent=-1.25, time=0.4268, value=0.98}
         <at 1 0 RCIM_Cubic>
        ]}
       BlueCurve {keys [<at 0.0643 5 RCIM_Cubic>,<at 0.1761 3.2823>,<at 0.3994 1>,<at 1 1 RCIM_Cubic>]}
       GreenCurve {keys [<at 0.0643 0 RCIM_Cubic>,<at 0.1761 0>,<at 0.3994 0>,<at 1 1 RCIM_Cubic>]}
       RedCurve {keys [<at 0.0643 3.0517 RCIM_Cubic>,<at 0.1761 5>,<at 0.3994 0.6397>,<at 1 1 RCIM_Cubic>]}
      }}))
   &CameraOffset(Camera Offset Amount=32) {module=CameraOffset001}
  ]
  rendererClasses [&NiagaraSpriteRendererProperties]
  renderers [{rendererClass=&NiagaraSpriteRendererProperties}]
 }
}
summary {
 systemName=NS_Emote_GhostHunter_GadgetSizzle
 emitters [
  {emitterName=Glitch001, rendererClasses [&NiagaraSpriteRendererProperties]}
  {emitterName=Empty, rendererClasses [&NiagaraMeshRendererProperties]}
  {emitterName=CubeEdge, rendererClasses [&NiagaraMeshRendererProperties]}
  {emitterName=Symbols, rendererClasses [&NiagaraSpriteRendererProperties]}
 ]
 userVariables [User.SkeletalMesh {
   type=NiagaraDataInterfaceSkeletalMesh
   defaultValue {
    struct=NiagaraExt_VariableValue_DataInterface
    value {dataInterface=&self:NiagaraDataInterfaceSkeletalMesh_0}
   }
  }]
}
