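# Niagara system for the GhostHunter emote's gadget warm-up: a purple sprite glow that charges up and ends in a short pink-violet flash.
# Glow bursts at start and lives for System.TimeToBurst, swelling, then fading and collapsing through color and non-uniform scale curves.
# BurstGlow spawns at that same TimeToBurst as a wider, brief pop, so one system variable syncs both emitters.
# Both apply a camera offset and expose no user parameters.
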NiagaraSystem /BRCosmetics/Animation/Game/MainPlayer/Emotes/GhostHunter/FX/NS_Emote_GhostHunter_GadgetWarmup
dependencies {
 usedDataInterfaces [&NiagaraDataInterfaceColorCurve, &NiagaraDataInterfaceCurve, &NiagaraDataInterfaceVector2DCurve]
 usedDynamicInputs []
 usedModules [
  &self:SystemSpawnScript.SystemScriptSource.SystemScriptGraph.NiagaraNodeAssignment_0.SetVariables_#1
  &SystemState
 ]
}
emitters {
 BurstGlow {
  emitterUpdateScript [
   &EmitterState()
   &SpawnBurst_Instantaneous(Spawn Time=$System.TimeToBurst, Use Spawn Probability=false, UseLoopCountLimit=false)
  ]
  particleSpawnScript [&InitializeParticle(
    Color=<rgba 1 0 0.8418 1>
    Lifetime=0.2
    Sprite Size=<100 52.0492>
    Sprite Size Mode=::NewEnumerator4
    UsePositionOffset=false
    Write Material Random=true
   )]
  particleUpdateScript [
   &ParticleState()
   &Color(Color=&ColorFromCurve(ColorCurve=DataInterface() {propertyValues {
       AlphaCurve {keys [<at 0 1 RCIM_Cubic>,<at 1 0 RCIM_Cubic>]}
       BlueCurve {keys [<at 0 5 RCIM_Cubic>,<at 0.2508 3>]}
       GreenCurve {keys [<at 0 0 RCIM_Cubic>,<at 0.2508 0.171>]}
       RedCurve {keys [<at 0 3.6697 RCIM_Cubic>,<at 0.2508 2.3709>]}
      }}))
   &ScaleSpriteSize(
    Non-Uniform Curve Sprite Scale=DataInterface() {propertyValues {
      XCurve {keys [
        {arriveTangent=9.3796, interpMode=RCIM_Cubic, leaveTangent=9.3796, tangentMode=RCTM_User, time=0, value=0}
        <at 0.2 2.5>
        <at 1 0 RCIM_Cubic>
       ]}
      YCurve {keys [
        {arriveTangent=9.3796, interpMode=RCIM_Cubic, leaveTangent=9.3796, tangentMode=RCTM_User, time=0, value=0}
        <at 0.2 2.5>
        {interpMode=RCIM_Cubic, leaveTangent=-6.6667, time=1, value=0}
       ]}
     }}
    Scale Sprite Size Mode=::NewEnumerator3
   )
   &CameraOffset(Camera Offset Amount=-10)
  ]
  rendererClasses [&NiagaraSpriteRendererProperties]
  renderers [{rendererClass=&NiagaraSpriteRendererProperties}]
 }
 Glow {
  emitterUpdateScript [
   &EmitterState()
   &SpawnBurst_Instantaneous(Use Spawn Probability=false, UseLoopCountLimit=false)
  ]
  particleSpawnScript [&InitializeParticle(
    Color Mode=::NewEnumerator0
    Lifetime=$System.TimeToBurst
    Sprite Size=<60 60>
    Sprite Size Mode=::NewEnumerator4
    UsePositionOffset=false
    Write Material Random=true
   )]
  particleUpdateScript [
   &ParticleState()
   &Color(Color=&ColorFromCurve(ColorCurve=DataInterface() {propertyValues {
       AlphaCurve {keys [<at 0.0144 0>,<at 0.2439 1 RCIM_Cubic>,<at 0.8511 1>,<at 1 0 RCIM_Cubic>]}
       BlueCurve {keys [<at 0.2508 3>,<at 0.8634 1.5>]}
       GreenCurve {keys [<at 0.2508 0>,<at 0.8634 0>]}
       RedCurve {keys [<at 0.2508 1.3888>,<at 0.8634 0.528>]}
      }}))
   &ScaleSpriteSize(
    Non-Uniform Curve Sprite Scale=DataInterface() {propertyValues {
      XCurve {keys [
        <at 0 0 RCIM_Cubic>
        {arriveTangent=2.7222, interpMode=RCIM_Cubic, leaveTangent=2.7222, time=0.4, value=1.375}
        {arriveTangent=-2.2917, interpMode=RCIM_Cubic, leaveTangent=-2.2917, time=0.9, value=2.45}
        <at 1 0 RCIM_Cubic>
       ]}
      YCurve {keys [
        <at 0 0 RCIM_Cubic>
        {arriveTangent=2.3611, interpMode=RCIM_Cubic, leaveTangent=2.3611, time=0.4, value=1.375}
        {arriveTangent=-2.2917, interpMode=RCIM_Cubic, leaveTangent=-2.2917, time=0.9, value=2.125}
        {interpMode=RCIM_Cubic, leaveTangent=-6.6667, time=1, value=0}
       ]}
     }}
    Scale Sprite Size Mode=::NewEnumerator3
   )
   &CameraOffset(Camera Offset Amount=-10)
  ]
  rendererClasses [&NiagaraSpriteRendererProperties]
  renderers [{rendererClass=&NiagaraSpriteRendererProperties}]
 }
}
summary {
 systemName=NS_Emote_GhostHunter_GadgetWarmup
 emitters [
  {emitterName=Glow, rendererClasses [&NiagaraSpriteRendererProperties]}
  {emitterName=BurstGlow, rendererClasses [&NiagaraSpriteRendererProperties]}
 ]
 userVariables []
}
