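# Niagara effect for the GhostHunter emote's swapping section, drawing a mesh wipe over the player for User.WipeDuration.
# One burst-spawned mesh particle follows System.CenterOfAnim with owner rotation and a camera offset.
# Its Y/Z scale tracks System.AnimBounds, while curves on normalized loop age swell and shrink X and fade white alpha in and out.
# The Empty emitter, another center-following mesh, is disabled.
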NiagaraSystem /BRCosmetics/Animation/Game/MainPlayer/Emotes/GhostHunter/FX/NS_Emote_GhostHunter_SwappingSection
dependencies {
 usedDataInterfaces [&NiagaraDataInterfaceCurve, &NiagaraDataInterfaceColorCurve]
 usedDynamicInputs [
  &self:ScratchDynamicInput_02
  &self:ScratchDynamicInput_01
 ]
 usedModules [
  &self:SystemSpawnScript.SystemScriptSource.SystemScriptGraph.NiagaraNodeAssignment_1.SetVariables_#1
  &SystemState
  &self:SystemSpawnScript.SystemScriptSource.SystemScriptGraph.NiagaraNodeAssignment_2.SetVariables_#2
  &self:SystemSpawnScript.SystemScriptSource.SystemScriptGraph.NiagaraNodeAssignment_3.SetVariables_#3
 ]
}
emitters {
 Empty {
  bEnabled=false
  emitterUpdateScript [
   &EmitterState()
   &SpawnBurst_Instantaneous(Use Spawn Probability=false, UseLoopCountLimit=false)
  ]
  particleSpawnScript [&InitializeParticle(Color Mode=::NewEnumerator0, Lifetime=5, UsePositionOffset=false, Write Material Random=true)]
  particleUpdateScript [
   &ParticleState()
   /SetVariables(Particles.Position:Vector3f=$System.CenterOfAnim) {isSetParameters=true}
   /SetVariables(Particles.MeshOrientation=$Engine.Owner.Rotation) {isSetParameters=true}
  ]
  rendererClasses [&NiagaraMeshRendererProperties]
  renderers [{rendererClass=&NiagaraMeshRendererProperties, rendererIndex=0}]
 }
 PlayerOverlayWipe {
  emitterUpdateScript [
   &EmitterState()
   &SpawnBurst_Instantaneous(Use Spawn Probability=false, UseLoopCountLimit=false)
  ]
  particleSpawnScript [&InitializeParticle(Lifetime=$User.WipeDuration, UsePositionOffset=false, Write Material Random=true)]
  particleUpdateScript [
   &ParticleState()
   /SetVariables(Particles.PrevPos:NiagaraPosition=$Particles.Position) {isSetParameters=true}
   /SetVariables(Particles.Position=&Subtract_Vector(A=$System.CenterOfAnim, B=<0 0 0>)) {isSetParameters=true}
   /SetVariables(Particles.Scale=&MakeVector(
     X=&FloatFromCurve(
      FloatCurve=DataInterface() {propertyValues {Curve {keys [
          <at 0 0 RCIM_Cubic>
          {arriveTangent=1.6667, interpMode=RCIM_Cubic, leaveTangent=1.6667, time=0.1, value=1}
          {arriveTangent=-1.2563, interpMode=RCIM_Cubic, leaveTangent=-1.2563, time=0.6, value=1}
          <at 0.896 0 RCIM_Cubic>
         ]}}}
      CurveIndex=$System.NormalizedLoopAge
      Scale Curve=30
     )
     Y=&MakeFloatFromVector(Vector=$System.AnimBounds, Channel=::NewEnumerator1)
     Z=&MakeFloatFromVector(Vector=$System.AnimBounds, Channel=::NewEnumerator2)
    )) {isSetParameters=true}
   /SetVariables(Particles.MeshOrientation=&MultiplyQuaternion(
     Quaternion A=$Engine.Owner.Rotation
     Quaternion B=&MakeQuatFromAxisAngle(Axis=<0 0 1>, AngleInDegrees=0)
    )) {isSetParameters=true}
   /SetVariables(Particles.MeshOrientation=&FindQuatBetweenTwoVectors(From Vector=$Particles.PrevPos, To Vector=Linked() {linkedVariable {name=Particles.Position, type=NiagaraPosition}})) {enabled=false, isSetParameters=true}
   &CameraOffset(Camera Offset Amount=-10)
   &Color(Color=&ColorFromCurve(
     ColorCurve=DataInterface() {propertyValues {
       AlphaCurve {keys [<at 0 0>,<at 0.1901 1 RCIM_Cubic>,<at 0.8095 1>,<at 1 0 RCIM_Cubic>]}
       BlueCurve {keys [<at 0 1 RCIM_Cubic>,<at 1 1 RCIM_Cubic>]}
       GreenCurve {keys [<at 0 1 RCIM_Cubic>,<at 1 1 RCIM_Cubic>]}
       RedCurve {keys [<at 0 1 RCIM_Cubic>,<at 1 1 RCIM_Cubic>]}
      }}
     CurveIndex=$System.NormalizedLoopAge
    ))
  ]
  rendererClasses [&NiagaraMeshRendererProperties]
  renderers [{rendererClass=&NiagaraMeshRendererProperties, rendererIndex=0}]
 }
}
summary {
 systemName=NS_Emote_GhostHunter_SwappingSection
 emitters [
  {emitterName=PlayerOverlayWipe, rendererClasses [&NiagaraMeshRendererProperties]}
  {bEnabled=false, emitterName=Empty, rendererClasses [&NiagaraMeshRendererProperties]}
 ]
 userVariables [
  User.SkeletalMesh {
   type=NiagaraDataInterfaceSkeletalMesh
   defaultValue {
    struct=NiagaraExt_VariableValue_DataInterface
    value {dataInterface=&self:NiagaraDataInterfaceSkeletalMesh_0}
   }
  }
  User.WipeDuration {type=NiagaraFloat, defaultValue=3.42, description=""}
 ]
}
