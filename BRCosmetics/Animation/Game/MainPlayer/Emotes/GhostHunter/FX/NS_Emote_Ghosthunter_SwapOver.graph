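# Niagara flash effect for the GhostHunter player emote, named for its swap-over moment and drawn with a mesh renderer.
# One emitter burst-spawns mesh particles whose lifetime, like the emitter loop, follows User.FlashDuration.
# Each update pins the mesh to the owner's position and rotation with a -90 degree Z turn; the curve-driven forward offset lerps between zeros, so adds nothing.
# A color curve shifts from white through cyan and magenta to violet as alpha fades out.
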NiagaraSystem /BRCosmetics/Animation/Game/MainPlayer/Emotes/GhostHunter/FX/NS_Emote_Ghosthunter_SwapOver
dependencies {
 usedDataInterfaces [&NiagaraDataInterfaceCurve, &NiagaraDataInterfaceColorCurve]
 usedDynamicInputs []
 usedModules [&SystemState]
}
emitters {Empty {
  emitterUpdateScript [
   &EmitterState(Life Cycle Mode=::NewEnumerator1, Loop Behavior=::NewEnumerator1, Loop Delay=0.2, Loop Duration=$User.FlashDuration, UseLoopDelay=false)
   &SpawnBurst_Instantaneous(Use Spawn Probability=false, UseLoopCountLimit=false)
  ]
  particleSpawnScript [&InitializeParticle(
    Color Mode=::NewEnumerator0
    Lifetime=$User.FlashDuration
    Mesh Scale Mode=::NewEnumerator3
    UsePositionOffset=false
    Write Material Random=true
   )]
  particleUpdateScript [
   &ParticleState()
   /SetVariables(Particles.MeshOrientation=&MultiplyQuaternion(
     Quaternion A=$Engine.Owner.Rotation
     Quaternion B=&MakeQuatFromAxisAngle(Axis=<0 0 1>, AngleInDegrees=-90)
    )) {isSetParameters=true}
   /SetVariables(Particles.Position=&Add_Vector(
     A=Linked() {linkedVariable {name=Engine.Owner.Position, type=NiagaraPosition}}
     B=&Multiply_VectorByFloat(
      Vector=&ReturnMeshOrientationAxis(Mesh Orientation Axis=::XAxis, Mesh Orientation Quat=$Particles.MeshOrientation)
      Float=&Lerp_Float(
       A=0
       B=0
       Alpha=&FloatFromCurve(FloatCurve=DataInterface() {propertyValues {Curve {keys [<at 0 0 RCIM_Cubic>,<at 1 1 RCIM_Cubic>]}}})
      )
     )
    )) {isSetParameters=true}
   &Color(Color=&ColorFromCurve(ColorCurve=DataInterface() {propertyValues {
       AlphaCurve {keys [{arriveTangent=-3.6859, interpMode=RCIM_Cubic, time=0, value=2}, <at 1 -0.25 RCIM_Cubic>]}
       BlueCurve {keys [<at 0 1 RCIM_Cubic>,<at 0.0481 1>,<at 0.2455 1>,<at 0.5291 1>]}
       GreenCurve {keys [<at 0 1 RCIM_Cubic>,<at 0.0481 0.8219>,<at 0.2455 0>,<at 0.5291 0>]}
       RedCurve {keys [<at 0 1 RCIM_Cubic>,<at 0.0481 0>,<at 0.2455 0.9504>,<at 0.5291 0.3251>]}
      }}))
  ]
  rendererClasses [&NiagaraMeshRendererProperties]
  renderers [{rendererClass=&NiagaraMeshRendererProperties, rendererIndex=0}]
 }}
summary {
 systemName=NS_Emote_Ghosthunter_SwapOver
 emitters [{emitterName=Empty, rendererClasses [&NiagaraMeshRendererProperties]}]
 userVariables [User.FlashDuration {type=NiagaraFloat, defaultValue=0.5, description=""}]
}
